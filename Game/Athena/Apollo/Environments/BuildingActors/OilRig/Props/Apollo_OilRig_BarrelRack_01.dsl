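; Blueprint for an oil rig barrel rack building prop in the Apollo environment that can show dripping oil.
; On BeginPlay it runs the parent logic, then checks the Is_Oil_Dripping flag.
; When the flag is set, it spawns the short oil drip Niagara effect attached to its static mesh component.
; The construction script only calls the parent.
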(event EventBeginPlay
  (|CallParentFunction)
  (if (Variables|Default|GetIs_Oil_Dripping)
    (Effects|SpawnFXSystemAttached "/Game/Effects/Fort_Effects/Effects/Environment/Apollo/NS_OilDrip_Short.NS_OilDrip_Short" (Variables|Debug|GetStaticMeshComponent))))

(fn ConstructionScript
  (|CallParentFunction))
