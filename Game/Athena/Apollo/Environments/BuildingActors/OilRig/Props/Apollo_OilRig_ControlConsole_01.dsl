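; Oil rig control console building prop for the Apollo map that quiets its looping screen audio when destroyed.
; On death, it validity-checks its two screen loop audio components and, when both are valid, briefly fades out loop 02.
; Loop 01 is only checked and never faded in this graph, despite the comment about stopping looping audio.
; The construction script is empty.
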(event EventOnDeathPlayEffects
  (params (Damage :type Float%20%28single-precision%29) (DamageTags :type Gameplay%20Tag%20Container%20Structure%20%28by%20ref%29) (Momentum :type Vector :default "0, 0, 0") (HitInfo :type Hit%20Result%20Structure%20%28by%20ref%29) (InstigatedBy :type Fort%20Pawn%20Object%20Reference) (DamageCauser :type Actor%20Object%20Reference) (EffectContext :type Gameplay%20Effect%20Context%20Handle%20Structure))
  (bind _audio_oilrig_controlconsole_screen_loop_01_cue (Variables|Default|GetAudio_OilRig_ControlConsole_Screen_Loop_01_Cue))
  (bind _audio_oilrig_controlconsole_screen_loop_02_cue (Variables|Default|GetAudio_OilRig_ControlConsole_Screen_Loop_02_Cue))
  ;[1 1 1] On Death Stop Looping Audio
  (Utilities|IsValid _audio_oilrig_controlconsole_screen_loop_02_cue
    (:Is%20Valid
      _)
    (:Is%20Not%20Valid
      _))
  (Utilities|IsValid _audio_oilrig_controlconsole_screen_loop_01_cue
    (:Is%20Valid
      (if (and (Utilities|IsValid _audio_oilrig_controlconsole_screen_loop_01_cue :!class /Script/Engine.KismetSystemLibrary) (Utilities|IsValid _audio_oilrig_controlconsole_screen_loop_02_cue :!class /Script/Engine.KismetSystemLibrary))
        (Audio|Components|Audio|FadeOut _audio_oilrig_controlconsole_screen_loop_02_cue 0.2 0.0 "Linear" :!class /Script/Engine.AudioComponent)))))

(fn ConstructionScript)
