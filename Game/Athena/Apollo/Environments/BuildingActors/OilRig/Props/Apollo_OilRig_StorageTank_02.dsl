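; Blueprint for an oil rig storage tank prop building actor that can drip oil and stops its ambient audio when destroyed.
; On begin play, a boolean flag decides whether a Niagara oil drip effect is attached to the static mesh at a fixed offset.
; The death effects event checks that the tank's ambience audio component is valid, then quickly fades it out.
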(event EventBeginPlay
  (|CallParentFunction)
  (if (Variables|Default|GetIs_Oil_Dripping)
    (Effects|SpawnFXSystemAttached "/Game/Effects/Fort_Effects/Effects/Environment/Apollo/NS_OilDrip_Short.NS_OilDrip_Short" (Variables|Debug|GetStaticMeshComponent) "None" "0.000000,51.000000,62.000000")))

(event EventOnDeathPlayEffects
  (params (Damage :type Float%20%28single-precision%29) (DamageTags :type Gameplay%20Tag%20Container%20Structure%20%28by%20ref%29) (Momentum :type Vector :default "0, 0, 0") (HitInfo :type Hit%20Result%20Structure%20%28by%20ref%29) (InstigatedBy :type Fort%20Pawn%20Object%20Reference) (DamageCauser :type Actor%20Object%20Reference) (EffectContext :type Gameplay%20Effect%20Context%20Handle%20Structure))
  (bind _audio_storagetank_ambience (Variables|Default|GetAudio_StorageTank_Ambience))
  (Utilities|IsValid _audio_storagetank_ambience
    (:Is%20Valid
      (if (Utilities|IsValid _audio_storagetank_ambience :!class /Script/Engine.KismetSystemLibrary)
        (Audio|Components|Audio|FadeOut _audio_storagetank_ambience 0.2 0.0 "Linear" :!class /Script/Engine.AudioComponent)))))

(fn ConstructionScript
  (|CallParentFunction))
